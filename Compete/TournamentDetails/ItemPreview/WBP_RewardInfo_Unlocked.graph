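# WBP_RewardInfo_Unlocked is an info panel in the tournament details item preview, telling the player a reward is a style for an item they do not yet own.
# A grid layers a horizontal box over a dark UIKit backplate, pairing a lock icon with auto-wrapping grey message text.
# The widget is static: it has no variables and its event graph holds only unconnected default events.
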WidgetBlueprint /Compete/TournamentDetails/ItemPreview/WBP_RewardInfo_Unlocked
graphs [EventGraph {nodes [
   Event PreConstruct()
   Event Construct()
   Event Tick()
  ]}]
variables []
widgets {
 info {inheritedWidgetCount=0, parentClass=&UserWidget, rootWidgetClass=&GridPanel, widgetCount=5}
 widgets [
  GP_Content {widgetClassPath=&GridPanel, widget=GP_Content, properties {columnFill [1], rowFill [1]}}
  WBP_UIKit_Backplate {
   widgetClassPath=&WBP_UIKit_Backplate
   parent=GP_Content
   widget=WBP_UIKit_Backplate
   properties {brightness=Dark}
   slot {class=GridSlot}
  }
  HorizontalBox_156 {
   widgetClassPath=&HorizontalBox
   parent=GP_Content
   widget=HorizontalBox_156
   slot {class=GridSlot, properties {layer=1, padding=<ltrb 8 8 12 8>}}
  }
  Image_Lock {
   widgetClassPath=&Image
   parent=HorizontalBox_156
   widget=Image_Lock
   properties {
    visibility=SelfHitTestInvisible
    brush {
     bIsDynamicallyLoaded=false
     drawAs=Image
     imageType=FullColor
     mirroring=NoMirror
     resourceName=None
     resourceObject=/FortUILibrary/Textures/MaterialInstances/MI_UIKit_Icon_Locked
     tiling=NoTile
     imageSize=<56 56>
     margin=<ltrb 0 0 0 0>
     outlineSettings {
      bUseBrushTransparency=false
      roundingType=HalfHeightRadius
      width=0
      color {colorUseRule=UseColor_Specified, specifiedColor=<rgba 0 0 0 0>}
      cornerRadii=<0 0 0 0>
     }
     tintColor {colorUseRule=UseColor_Specified, specifiedColor=<rgba 1 1 1 1>}
     uVRegion {bIsValid=false, max=<0 0>, min=<0 0>}
    }
   }
   slot {class=HorizontalBoxSlot, properties {horizontalAlignment=HAlign_Left, verticalAlignment=VAlign_Top, padding=<ltrb 0 0 4 0>}}
  }
  Text_UnownedBaseItemMessage {
   widgetClassPath=&CommonTextBlock
   parent=HorizontalBox_156
   widget=Text_UnownedBaseItemMessage
   properties {
    autoWrapText=true
    text=The style for an item that you do not yet own.
    colorAndOpacity {colorUseRule=UseColor_Specified, specifiedColor=<rgba 0.5776 0.5776 0.5776 1>}
    font {
     bForceMonospaced=false
     bMaterialIsStencil=false
     fontMaterial=null
     fontObject=&NotoSans
     letterSpacing=0
     monospacedWidth=1
     size=24
     skewAmount=0
     typefaceFontName=NotoSans-Regular
     outlineSettings {
      bApplyOutlineToDropShadows=false
      bMiteredCorners=false
      bSeparateFillAlpha=false
      outlineMaterial=null
      outlineSize=0
      outlineColor=<rgba 0 0 0 1>
     }
    }
    shadowOffset=<0 0>
   }
   slot {class=HorizontalBoxSlot, properties {verticalAlignment=VAlign_Top, size {sizeRule=Fill, value=1}}}
  }
 ]
}
